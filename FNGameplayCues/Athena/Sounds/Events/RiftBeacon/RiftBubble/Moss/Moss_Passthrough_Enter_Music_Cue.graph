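# Sound cue that plays music when entering the Moss rift bubble, part of the Rift Beacon event sounds.
# A mixer layers one of two random enter stingers, classed as gadgets, with a quieter looping Fortnitemares night ambience track, classed as music packs.
# An enveloper fades the loop in over seven seconds.
# It borrows the safe zone shield attenuation and allows one instance per owner, stopping the oldest.
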SoundCue /FNGameplayCues/Athena/Sounds/Events/RiftBeacon/RiftBubble/Moss/Moss_Passthrough_Enter_Music_Cue
root=Mixer_0
nodes [
 {
  id=Enveloper_0
  children [WavePlayer_0]
  properties {pitchCurve {editorCurveData {keys [{time=0}]}}, volumeCurve {editorCurveData {keys [<at 0 0>,<at 7 1>]}}}
 }
 {id=Mixer_0, children [SoundClass_0,SoundClass_1], properties {inputVolume [1,0.3]}}
 {id=Random_2, children [WavePlayer_5,WavePlayer_6], properties {weights [1,1]}}
 {id=SoundClass_0, children [Random_2], properties {soundClassOverride=&Fort_Gadgets}}
 {id=SoundClass_1, children [Enveloper_0], properties {soundClassOverride=/Game/Sounds/FortSoundClassesAndModes/Fort_Music_MusicPacks}}
 {
  id=WavePlayer_0
  properties {bLooping=true, soundWaveAssetPtr=/FNGameplayCues/Sounds/Fort_Music/Orchestral_Score/Hexylvania/Fortnitemares_Amb_Night_Music_02_Loop}
 }
 {id=WavePlayer_5, properties {soundWaveAssetPtr=&/Moss_Music_Enter_02}}
 {id=WavePlayer_6, properties {soundWaveAssetPtr=&/Moss_Music_Enter_01}}
]
properties {
 attenuationSettings=/Game/Sounds/Attenuation/Fort_SafeZone_Shield_Athena_Att
 bOverrideConcurrency=true
 maxDistance=4000
 volumeMultiplier=1
 concurrencyOverrides {bLimitToOwner=true, maxCount=1, resolutionRule=StopOldest}
}
